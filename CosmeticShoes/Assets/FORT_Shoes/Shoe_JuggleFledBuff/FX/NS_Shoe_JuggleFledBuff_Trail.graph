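# Niagara trail for the JuggleFledBuff cosmetic shoes that sheds green leaf sprites from the player's feet, tinted by time of day.
# Spawn rate ramps with owner speed, drops to zero when hidden or skydiving, and has a front-end preview override.
# Partitioned particles sample per-foot skeletal bones or sockets, killed per partition by shoe scale checks.
# Curves blend them from foot-attached to free drift under drag and curl noise, with separate front-end motion.
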NiagaraSystem /CosmeticShoes/Assets/FORT_Shoes/Shoe_JuggleFledBuff/FX/NS_Shoe_JuggleFledBuff_Trail
dependencies {
 usedDataInterfaces [&NiagaraDataInterfaceCurve, &NiagaraDataInterfaceSkeletalMesh]
 usedDynamicInputs [
  /Game/Effects/Niagara/DynamicInputs/TimeOfDay/DaySequenceTimeOfDayColor
  &FN_Select_Color_from_Time_Of_Day
 ]
 usedModules [
  &self:SystemSpawnScript.SystemScriptSource.SystemScriptGraph.NiagaraNodeAssignment_3.SetVariables_#1
  &SystemState
  &self:SystemSpawnScript.SystemScriptSource.SystemScriptGraph.NiagaraNodeAssignment_0.SetVariables_#2
 ]
}
emitters {Leafs {
  emitterUpdateScript [
   &EmitterState(Life Cycle Mode=::NewEnumerator1, Loop Behavior=::NewEnumerator1, Loop Duration Mode=::NewEnumerator1, UseLoopDelay=false)
   /SetVariables(Emitter.VelocityLerp:NiagaraFloat=&FloatFromCurve(
     FloatCurve=DataInterface() {propertyValues {Curve {keys [<at 50 0>,<at 200 1>]}}}
     CurveIndex=&VectorLength(Vector=$Engine.Owner.Velocity)
    )) {isSetParameters=true}
   &SpawnRate(
    SpawnRate=&Lerp_Float(
     A=0
     B=&MakeCustomFloatFromBool(Bool=&Boolean_OrOperation(Bool1=$User.HideFootFX, Bool2=$User.IsSkydiving), True Float=0, False Float=50)
     Alpha=&FN_FrontEndFloatModifier(
      Gameplay=$Emitter.VelocityLerp
      Front End=&MakeCustomFloatFromBool(Bool=$User.bisPrimaryPreviewItem, True Float=0.5, False Float=0)
     )
    )
    Use Spawn Probability=false
   ) {module=SpawnRate001}
  ]
  particleSpawnScript [
   &PartitionParticles(Write Partition to Ribbon ID=::NewEnumerator0)
   &InitializeParticle(
    Color=&Multiply_LinearColorByFloat(
     Linear Color=&RandomRangeLinearColor(Minimum=<rgba 0.3896 1 0.0065 1>, Maximum=<rgba 0.2442 0.6 0.003 1>)
     Float=&FN_FrontEndFloatModifier(Gameplay=1, Front End=0.4)
    )
    Lifetime Max=1
    Lifetime Min=0.5
    Lifetime Mode=::NewEnumerator1
    Sprite Rotation Mode=::NewEnumerator2
    Sprite Size Mode=::NewEnumerator3
    Uniform Sprite Size=&MakeFloatFromVector2d(Vector 2D=&Multiply_Vector2D_ByFloat(
      Float=&RandomRangeFloat(Minimum=0.15, Maximum=0.3)
      Vector2D=&RandomRangeVector2D(Minimum=<30 20>, Maximum=<40 45>)
     ))
    UsePositionOffset=false
    Write Material Random=true
   )
   &AddVelocity(
    Rotation Coordinate Space=::World
    Velocity=&Add_Vector(
     A=&Multiply_VectorByFloat(Vector=&RandomRangeVector(Minimum=<-1 -1 0>, Maximum=<1 1 0.5>), Float=200)
     B=&Multiply_Vector(
      A=&MakeCustomVectorFromBool(
       Bool=$User.bisPrimaryPreviewItem
       True Vector=&RotateVectorByQuaternion(
        Vector To Rotate=/Niagara/DynamicInputs/Vectors/RotateVector(
         Vector To Rotate:Vector3f=&RandomRangeVector(Minimum=<0 400 0>, Maximum=<0 700 0>)
         Yaw:NiagaraFloat=20
         Pitch:NiagaraFloat=0
         Roll:NiagaraFloat=&RandomRangeFloat(Minimum=0, Maximum=5)
         Rotation Rate:NiagaraFloat=1
        )
        Quaternion=$Engine.Owner.Rotation
       )
       False Vector=<0 0 0>
      )
      B=<-1 -1 -1>
     )
    )
    Velocity Speed Scale=&RandomRangeFloat(Minimum=0.2, Maximum=0.3)
   )
   &InitialMeshOrientation(Use Rotation Vector=false)
   /SetVariables() {isSetParameters=true}
  ]
  particleUpdateScript [
   &KillParticles(Kill Particles=&Boolean_OrOperation(
     Bool1=&Boolean_AndOperation(
      Bool1=&SetBoolByFloatComparison(
       A=&MakeFloatFromVector(Vector=$User.ShoeScale_R, Channel=::NewEnumerator0)
       B=0.5
       Comparison Type=::NewEnumerator6
      )
      Bool2=&SetBoolByIntComparison(A=$Particles.PartitionParticles.Partition, B=0, Comparison Type=::NewEnumerator4)
     )
     Bool2=&Boolean_AndOperation(
      Bool1=&SetBoolByFloatComparison(
       A=&MakeFloatFromVector(Vector=$User.ShoeScale_L, Channel=::NewEnumerator0)
       B=0.5
       Comparison Type=::NewEnumerator6
      )
      Bool2=&SetBoolByIntComparison(A=$Particles.PartitionParticles.Partition, B=1, Comparison Type=::NewEnumerator4)
     )
    ))
   &ParticleState()
   &SpriteRotationRate(Rotation Rate=120)
   &SkeletalMeshLocation(
    Bone / Socket Index=$Particles.PartitionParticles.Partition
    Bone Sampling Mode=::NewEnumerator4
    MaskSpawnGroup=false
    Orientation Sampling=::NewEnumerator1
    Position Sampling=::NewEnumerator2
    Sampled Position Offset=&Multiply_Vector(
     A=&RandomRangeVector(Minimum=<4 -20 0>, Maximum=<0 10 1>)
     B=&MergeHelper_VectorSwitch(0=<-1 1 0>, 1=<1 -1 0>, 2=?, 3=?, Index=$Particles.PartitionParticles.Partition, Num=::NewEnumerator0)
    )
    Skeletal Mesh=$System.Skeletal Mesh
   )
   &Drag(
    Drag=&FloatFromCurve(
     FloatCurve=DataInterface() {propertyValues {Curve {keys [{arriveTangent=3.4423, interpMode=RCIM_Cubic, leaveTangent=3.4423, tangentMode=RCTM_User, time=0, value=0}, <at 1 1 RCIM_Cubic>]}}}
     Scale Curve=10
    )
    Use Linear Drag=true
    Use Rotational Drag=true
   )
   &CurlNoiseForce(
    Noise Frequency=&RandomRangeFloat(Minimum=5, Maximum=15)
    Noise Strength=&FloatFromCurve(
     FloatCurve=DataInterface() {propertyValues {Curve {
        defaultValue=1
        keys [{interpMode=RCIM_Cubic, tangentMode=RCTM_User, time=0, value=0}, {interpMode=RCIM_Cubic, tangentMode=RCTM_User, time=0.5, value=1}]
       }}}
     Scale Curve=250
    )
    Pan Noise=false
   )
   &SolveForcesAndVelocity(Clamp Velocity=false, Limit Acceleration=false)
   &ScaleColor(Scale RGB=&MakeVectorFromLinearColorRGB(Color=$System.TODColor), ScaleA=true, ScaleRGB=true)
   /SetVariables(Particles.FrameVelocity:Vector3f=&Multiply_Vector(A=$Particles.Velocity, B=&VectorFromFloat(Value=$Engine.DeltaTime))) {isSetParameters=true}
   /SetVariables(Particles.VelocityOffset:Vector3f=&Add_Vector(A=$Particles.VelocityOffset, B=$Particles.FrameVelocity)) {isSetParameters=true}
   /SetVariables(Particles.Position=&FN_FrontEndVector3Modifier(
     Gameplay=&Add_Vector(
      A=&Lerp_Vector(
       A=&Add_Vector(A=Linked() {linkedVariable {name=Particles.SkeletalMeshLocation.SampledPosition, type=NiagaraPosition}}, B=$Particles.VelocityOffset)
       B=&Add_Vector(
        A=Linked() {linkedVariable {name=Particles.Position, type=NiagaraPosition}}
        B=&Multiply_VectorByFloat(Vector=$Particles.FrameVelocity, Float=-3)
       )
       Alpha=&FloatFromCurve(
        FloatCurve=DataInterface() {propertyValues {Curve {defaultValue=1, keys [<at 0.1 0 RCIM_Cubic>, {interpMode=RCIM_Cubic, leaveTangent=3.0199, time=0.75, value=1}]}}}
        CurveIndex=&Add_Float(A=&RandomRangeFloat(Minimum=0, Maximum=0), B=$Particles.NormalizedAge)
       )
      )
      B=<0 0 0>
     )
     Front End=&Add_Vector(
      A=&Lerp_Vector(
       A=&Add_Vector(A=Linked() {linkedVariable {name=Particles.SkeletalMeshLocation.SampledPosition, type=NiagaraPosition}}, B=$Particles.VelocityOffset)
       B=&Add_Vector(
        A=Linked() {linkedVariable {name=Particles.Position, type=NiagaraPosition}}
        B=&Multiply_VectorByFloat(Vector=$Particles.FrameVelocity, Float=1)
       )
       Alpha=&FloatFromCurve(
        FloatCurve=DataInterface() {propertyValues {Curve {defaultValue=1, keys [<at 0.1 0 RCIM_Cubic>,<at 0.2 1 RCIM_Cubic>]}}}
        CurveIndex=&Add_Float(A=&RandomRangeFloat(Minimum=0, Maximum=0), B=$Particles.NormalizedAge)
       )
      )
      B=<0 0 0>
     )
    )) {isSetParameters=true}
   &ScaleSpriteSize(Uniform Curve Sprite Scale=DataInterface() {propertyValues {Curve {keys [
        {interpMode=RCIM_Cubic, tangentMode=RCTM_User, time=0, value=1}
        {arriveTangent=-2.776, interpMode=RCIM_Cubic, leaveTangent=-2.776, tangentMode=RCTM_User, time=1, value=0}
       ]}}})
   &DynamicMaterialParameters(
    Index 0 Param 1=0.1
    Module.Index0Float0Write=true
    Module.Index0Float1Write=true
    Module.Index0Float2Write=true
    Module.Index0Float3Write=true
   )
  ]
  rendererClasses [&NiagaraSpriteRendererProperties]
  renderers [{rendererClass=&NiagaraSpriteRendererProperties}]
 }}
summary {
 systemName=NS_Shoe_JuggleFledBuff_Trail
 emitters [{emitterName=Leafs, rendererClasses [&NiagaraSpriteRendererProperties]}]
 userVariables [
  User.bIsFrontEnd {type=NiagaraBool}
  User.bIsFrontEndPreview {type=NiagaraBool}
  User.bisPrimaryPreviewItem {type=NiagaraBool}
  User.HideFootFX {type=NiagaraBool}
  User.isOffPawnPreview {type=NiagaraBool}
  User.IsSkydiving {type=NiagaraBool}
  User.ShoeScale_L {type=Vector3f, description="", defaultValue {struct=Vector3f, value=<1 1 1>}}
  User.ShoeScale_R {type=Vector3f, description="", defaultValue {struct=Vector3f, value=<1 1 1>}}
 ]
}
