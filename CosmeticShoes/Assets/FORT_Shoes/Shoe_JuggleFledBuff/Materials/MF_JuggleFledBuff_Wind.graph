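# Material function for the JuggleFledBuff cosmetic shoe that adds wind sway to the incoming material's World Position Offset, apparently for its leaves.
# Two time-driven sine waves over pre-skinning local position multiply together, pushing vertices along a blend of normal and wind vector.
# Vertex color packs the controls: R weights displacement, G offsets timing, B is rounded into leaf-type masks exposed as named reroutes.
# A wind mask texture fades the offset to zero.
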MaterialFunction /CosmeticShoes/Assets/FORT_Shoes/Shoe_JuggleFledBuff/Materials/MF_JuggleFledBuff_Wind
graph {nodes [
  n1 = GetMaterialAttributes(None=FunctionInput() {inputName=Material, inputType=FunctionInput_MaterialAttributes}) {attributeGetTypes [#1]}
  n2 = VertexNormalWS()
  n3 = Transform(None=Normalize(VectorInput=VectorParameter() {parameterName=FullGeoWindVector, defaultValue=<rgba 0.5 -0.5 0 1>}.RGB)) {transformSourceType=TRANSFORMSOURCE_Local}
  n4 = NamedRerouteUsage()
  n5 = Divide(
   A=LocalPosition() {localOrigin=InstancePreSkinning}.XYZ
   B=ScalarParameter() {defaultValue=2, group=JuggleFledBuff|Wind, parameterName=SinewaveFrequency}
  )
  n6 = VertexColor()
  n7 = ScalarParameter() {defaultValue=1, group=JuggleFledBuff|Wind, parameterName=RandomTimeOffset}
  n8 = Multiply(A=Add(A=Time(), B=Multiply(A=@n6.G, B=@n7)), B=ScalarParameter() {defaultValue=0.3, group=JuggleFledBuff|Wind, parameterName=Speed})
  n9 = Add(A=Sine(None=Add(A=@n5, B=@n8)))
  n10 = Divide(
   A=LocalPosition() {localOrigin=InstancePreSkinning}.XYZ
   B=ScalarParameter() {defaultValue=4, group=JuggleFledBuff|Wind, parameterName=SinewaveFrequency2}
  )
  n11 = VertexColor()
  n12 = ScalarParameter() {defaultValue=1, group=JuggleFledBuff|Wind, parameterName=RandomTimeOffset}
  n13 = Multiply(A=Add(A=Time(), B=Multiply(A=@n11.G, B=@n12)), B=ScalarParameter() {defaultValue=0.712, group=JuggleFledBuff|Wind, parameterName=Speed2})
  n14 = Add(A=Sine(None=Add(A=@n10, B=@n13)))
  n15 = VertexColor()
  n16 = Multiply(
   A=Multiply(A=LinearInterpolate(A=@n2, B=@n3, Alpha=@n4), B=Multiply(A=@n9, B=@n14))
   B=LinearInterpolate(A=@n15.R, B=Power(Base=@n15.R) {constExponent=0.25}, Alpha=NamedRerouteUsage())
  )
  n17 = ScalarParameter() {defaultValue=0.1, group=JuggleFledBuff|Wind, parameterName=Amplitude}
  n18 = LinearInterpolate(
   A=Multiply(A=Multiply(A=@n16, B=@n17), B=NamedRerouteUsage())
   B=Constant()
   Alpha=TextureSample() {samplerType=SAMPLERTYPE_Alpha, texture=/CosmeticShoes/Assets/FORT_Shoes/Shoe_JuggleFledBuff/Textures/T_JuggleFledBuff_WindMask}.R
  )
  FunctionOutput(None=SetMaterialAttributes!(MaterialAttributes=@n1.MaterialAttributes, WorldPositionOffset=Add(A=@n1.WorldPositionOffset, B=@n18)) {attributeSetTypes [#1]})
  n19 = Round(None=Multiply(A=VertexColor().B) {constB=3})
  NamedRerouteDeclaration(None=Saturate(None=@n19)) {name=AllLeavesMask, nodeColor=<rgba 0 0.2734 1 1>}
  n20 = Subtract(A=@n19)
  NamedRerouteDeclaration(None=Subtract(A=Saturate(None=@n19), B=Saturate(None=@n20))) {name=FullGeoLeavesMask, nodeColor=<rgba 0 1 0.7656 1>}
  n21 = Subtract(A=@n19)
  n22 = Subtract(A=@n19) {constB=2}
  NamedRerouteDeclaration(None=Subtract(A=Saturate(None=@n21), B=Saturate(None=@n22))) {name=BackFaceLeavesMask, nodeColor=<rgba 0 1 0.7656 1>}
  NamedRerouteDeclaration(None=Saturate(None=Subtract(A=@n19) {constB=2})) {name=FrontFaceLeavesMask, nodeColor=<rgba 0 1 0.7656 1>}
 ]}
