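# Parent skeletal mesh material for the JuggleFledBuff cosmetic shoe, supporting several shading models including cloth and clear coat.
# It chains character PreFX, a shoe-specific wind function, the shared sock function, character PostFX and a tech-art hologram function through material attributes.
# A UV step forces opacity mask to 1 where U reaches 1 or more, overriding the sock result.
# Final specular is clamped by a parameter within the cloth fuzz mask channel.
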Material /CosmeticShoes/Assets/FORT_Shoes/Shoe_JuggleFledBuff/Materials/M_JuggleFledBuff_Parent
graph {
 properties {
  bAutomaticallySetUsageInEditor=false
  bUseMaterialAttributes=true
  bUsedWithMeshDeformer=true
  bUsedWithMorphTargets=true
  bUsedWithSkeletalMesh=true
  pixelDepthOffsetMode=PDOM_Legacy
  previewMesh=/Game/Developers/luisbatista/JuggleFledBuff_Shoe_VC
  shadingModel=MSM_FromMaterialExpression
  usedShadingModels=Default Lit | Subsurface | Clear Coat | Subsurface Profile | Cloth
 }
 nodes [
  n1 = MaterialFunctionCall() {materialFunction=&MF_CharacterParent_2023_PreFX}
  n2 = MaterialFunctionCall!(
   InOpacity
   Material=MaterialFunctionCall!(Material=@n1.Result) {materialFunction=&/MF_JuggleFledBuff_Wind}.Result
   Mask=@n1.Tech Art Mask
  ) {materialFunction=/Game/Global/Materials/Cosmetics/Kicks/Socks/Materials/MF_Sock}
  n3 = TextureCoordinate()
  n4 = BlendMaterialAttributes(
   A=@n2.Result
   B=SetMaterialAttributes!(MaterialAttributes=@n2.Result, Opacity Mask=Constant() {r=1}) {attributeSetTypes [#1]}
   Alpha=Step(X=ComponentMask(None=@n3) {r=true}) {constY=1}
  )
  n5 = GetMaterialAttributes(None=MaterialFunctionCall!(In=MaterialFunctionCall(Material=@n4, M=@n1.M, Pre Skinned Local Position=@n1.M, Distance Blend=@n1.M, Fresnel=@n1.M) {materialFunction=&MF_CharacterParent_2023_PostFX}.Result) {materialFunction=/Game/Global/Materials/Cosmetics/Kicks/Shoes/Materials/MF_TechartShoeHologram}.Result) {attributeGetTypes [#2, #3]}
  n6 = ScalarParameter() {defaultValue=1, group=04_Tech Art, parameterName=Specular Clamp}
  SetMaterialAttributes!(
   MaterialAttributes=@n5.MaterialAttributes
   Specular=LinearInterpolate(
    A=@n5.Specular
    B=Clamp(None=@n5.Specular, Max=@n6)
    Alpha=ChannelMaskParameter(None=@n1.Tech Art Mask) {group=05_Tech Art Masks, parameterName=ClothFuzzMaskChannel}.R
   )
   Anisotropy
  ) {attributeSetTypes [#2, #3]}
 ]
}
